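# Physics asset for the side hair of the F_MED Desert Ops Camo female character's face accessory, covering two left-side dynamic strands, A and B.
# Each strand is a four-capsule chain from base through parent, 01 and 02, with free swing and twist at the head and narrow angular limits between segments.
# Capsules on the head, neck, upper spine and left forearm supply body shapes around the strands.
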PhysicsAsset /BRCosmetics/Characters/Player/Female/Medium/Bodies/F_MED_Desert_Ops_Camo/Meshes/Parts/Hair/F_MED_DesOpsCamo_Sides_FaceAcc_Physics
bodies {
 dyn_hair_sideA_l_01 [{shapeType=Capsule, length=2.1214, radius=1.5856, center=<0.4582 0 0>, rotation=<rot -90 -0.0396 0.0397>}]
 dyn_hair_sideA_l_02 [{shapeType=Capsule, length=2.0596, radius=1.2912, center=<1.8991 0.3484 0.2032>, rotation=<rot 0 104.1839 -81.1195>}]
 dyn_hair_sideA_l_base [{shapeType=Capsule, length=0.505, radius=0.505, rotation=<rot 0 -90 0>}]
 dyn_hair_sideA_l_parent [{shapeType=Capsule, length=2, radius=2.005, center=<0.1854 -0.0001 0>, rotation=<rot -90 0.0264 -0.0263>}]
 dyn_hair_sideB_l_01 [{shapeType=Capsule, length=2.1214, radius=1.5856, center=<0.4582 0 0>, rotation=<rot -90 -0.0396 0.0397>}]
 dyn_hair_sideB_l_02 [{shapeType=Capsule, length=2.0596, radius=1.2912, center=<1.8991 0.3484 0.2032>, rotation=<rot 0 104.1839 -81.1195>}]
 dyn_hair_sideB_l_base [{shapeType=Capsule, length=0.505, radius=0.505, rotation=<rot 0 -90 0>}]
 dyn_hair_sideB_l_parent [{shapeType=Capsule, length=2, radius=2.005, center=<0.1854 -0.0001 0>, rotation=<rot -90 0.0264 -0.0263>}]
 head [{shapeType=Capsule, length=5, radius=6.75, rotation=<rot 90 -0.0003 -0.0002>}]
 lowerarm_l [{shapeType=Capsule, length=25, radius=6.75, center=<-11.6871 5.577 1.3776>, rotation=<rot 77.1264 36.7567 65.5791>}]
 neck_01 [{shapeType=Capsule, length=6.25, radius=4.5, center=<0.3642 1.1246 0.0048>, rotation=<rot 89.3571 -179.9998 -162.0508>}]
 spine_04 [{shapeType=Capsule, length=12.25, radius=17.75, center=<-6.5908 -5.4418 -0.9616>, rotation=<rot 89.3251 179.9999 180>}]
 spine_05 [{shapeType=Capsule, length=15, radius=8.25, center=<2.3703 -1.851 0.0001>}]
}
constraints [
 {
  bone1Name=dyn_hair_sideA_l_base
  bone2Name=head
  swing1LimitDegrees=45
  swing1Motion=Free
  swing2LimitDegrees=45
  swing2Motion=Free
  twistLimitDegrees=45
  twistMotion=Free
 }
 {bone1Name=dyn_hair_sideA_l_parent, bone2Name=dyn_hair_sideA_l_base, swing1LimitDegrees=40, swing2LimitDegrees=25, twistLimitDegrees=5}
 {bone1Name=dyn_hair_sideA_l_01, bone2Name=dyn_hair_sideA_l_parent, swing1LimitDegrees=25, swing2LimitDegrees=25, twistLimitDegrees=5}
 {bone1Name=dyn_hair_sideA_l_02, bone2Name=dyn_hair_sideA_l_01, swing1LimitDegrees=25, swing2LimitDegrees=25, twistLimitDegrees=5}
 {
  bone1Name=dyn_hair_sideB_l_base
  bone2Name=head
  swing1LimitDegrees=45
  swing1Motion=Free
  swing2LimitDegrees=45
  swing2Motion=Free
  twistLimitDegrees=45
  twistMotion=Free
 }
 {bone1Name=dyn_hair_sideB_l_parent, bone2Name=dyn_hair_sideB_l_base, swing1LimitDegrees=40, swing2LimitDegrees=25, twistLimitDegrees=5}
 {bone1Name=dyn_hair_sideB_l_01, bone2Name=dyn_hair_sideB_l_parent, swing1LimitDegrees=25, swing2LimitDegrees=25, twistLimitDegrees=5}
 {bone1Name=dyn_hair_sideB_l_02, bone2Name=dyn_hair_sideB_l_01, swing1LimitDegrees=25, swing2LimitDegrees=25, twistLimitDegrees=5}
]
