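# Physics asset for the ponytail on the female medium Desert Ops Camo cosmetic's hair and face accessory mesh.
# Capsule bodies form a constraint chain from spine_05 through head and dyn_parent_pony down five ponytail bones, with large head and upper-spine volumes.
# The three upper joints rotate freely.
# The four lower joints have swing limits and tight twist limits, with capsules narrowing from dyn_pony_02 toward the tip.
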PhysicsAsset /BRCosmetics/Characters/Player/Female/Medium/Bodies/F_MED_Desert_Ops_Camo/Meshes/Parts/Hair/F_MED_Desert_Ops_Camo_Ponytail_FaceAcc_Physics
bodies {
 dyn_parent_pony [{shapeType=Capsule, length=0.505, radius=0.505, rotation=<rot 0 -90 0>}]
 dyn_pony_01 [{shapeType=Capsule, length=1, radius=1}]
 dyn_pony_02 [{shapeType=Capsule, length=5.75, radius=3.25, center=<5 0 0>, rotation=<rot -90 -0.422 0.4221>}]
 dyn_pony_03 [{shapeType=Capsule, length=4.25, radius=2.5, center=<5 0 0>, rotation=<rot -90 -178.1016 178.1016>}]
 dyn_pony_04 [{shapeType=Capsule, length=3, radius=2, center=<5 0 0>, rotation=<rot -90 -0.0049 0.0049>}]
 dyn_pony_05 [{shapeType=Capsule, length=3.25, radius=1.25, center=<3.7292 0 0>, rotation=<rot -90 0.0264 -0.0263>}]
 head [{shapeType=Capsule, length=7.5, radius=6.75, center=<3.5252 0.0924 0>, rotation=<rot -90 0 -15.0002>}]
 spine_05 [{shapeType=Capsule, length=12.505, radius=13.005, center=<0 3.2538 0>, rotation=<rot 0 -90 0>}]
}
constraints [
 {bone1Name=dyn_pony_02, bone2Name=dyn_pony_01, swing1LimitDegrees=65, swing2LimitDegrees=30, twistLimitDegrees=8}
 {bone1Name=dyn_pony_03, bone2Name=dyn_pony_02, swing1LimitDegrees=45, swing2LimitDegrees=35, twistLimitDegrees=5}
 {bone1Name=dyn_pony_04, bone2Name=dyn_pony_03, swing1LimitDegrees=45, swing2LimitDegrees=29.5714, twistLimitDegrees=5}
 {bone1Name=dyn_pony_05, bone2Name=dyn_pony_04, swing1LimitDegrees=45, swing2LimitDegrees=31.2857, twistLimitDegrees=5}
 {
  bone1Name=dyn_pony_01
  bone2Name=dyn_parent_pony
  swing1LimitDegrees=45
  swing1Motion=Free
  swing2LimitDegrees=45
  swing2Motion=Free
  twistLimitDegrees=45
  twistMotion=Free
 }
 {
  bone1Name=head
  bone2Name=spine_05
  swing1LimitDegrees=45
  swing1Motion=Free
  swing2LimitDegrees=45
  swing2Motion=Free
  twistLimitDegrees=45
  twistMotion=Free
 }
 {
  bone1Name=dyn_parent_pony
  bone2Name=head
  swing1LimitDegrees=45
  swing1Motion=Free
  swing2LimitDegrees=45
  swing2Motion=Free
  twistLimitDegrees=45
  twistMotion=Free
 }
]
