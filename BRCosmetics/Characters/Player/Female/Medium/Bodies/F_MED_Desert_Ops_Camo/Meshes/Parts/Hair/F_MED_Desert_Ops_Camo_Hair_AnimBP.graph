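# Post-process style hair AnimBP for the F_MED_Desert_Ops_Camo character, adding secondary motion to the pose copied from the attached parent mesh.
# On customization load, meta tags select the hairstyle and set hat and vest alphas.
# An int blend picks plain copy, long hair (LookAt, RigidBody, bone offsets), AnimDynamics bangs, or RigidBody ponytail, bypassed under hats.
# The pony AnimDynamics chain and Control Rig are unconnected to the output.
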AnimBlueprint /BRCosmetics/Characters/Player/Female/Medium/Bodies/F_MED_Desert_Ops_Camo/Meshes/Parts/Hair/F_MED_Desert_Ops_Camo_Hair_AnimBP
target_skeleton=&/F_MED_Desert_Ops_Camo_Hair_Skeleton
graphs [
 AnimGraph {nodes [
   n1 = Anim Dynamics - Bone: dyn_pony_01(Alpha=1, ExternalForce="0,0,0") {
    angularDampingOverride=1
    angularSpringConstant=13000
    bAngularSpring=true
    bGravityOverrideInSimSpace=true
    bLinearSpring=true
    bOverrideAngularDamping=true
    bUseGravityOverride=true
    gravityScale=0.35
    linearSpringConstant=100
    alphaBoolBlend {blendOption=Linear}
    alphaScaleBiasClamp {scale=0.3}
    boundBone {boneName=dyn_pony_01}
    componentAppliedLinearAccClamp=<1100 1100 1100>
    componentLinearAccScale=<1 1 1>
    componentLinearVelScale=<1 1 1>
    gravityOverride {z=-900}
    physicsBodyDefinitions [{
      sphereCollisionRadius=0
      boundBone {boneName=dyn_pony_01}
      boxExtents=<8 10 10>
      constraintSetup {angularLimitsMax=<15 15 45>, angularLimitsMin=<-15 -35 -45>, angularTarget=<1 0 0>}
      localJointOffset=<3 0 0>
     }]
   }
   n2 = Anim Dynamics - Bone: dyn_pony_02(ComponentPose=@n1, Alpha=1, ExternalForce="0,0,0") {
    angularDampingOverride=1
    angularSpringConstant=1500
    bAngularSpring=true
    bGravityOverrideInSimSpace=true
    bOverrideAngularDamping=true
    bUseGravityOverride=true
    gravityScale=0.35
    linearSpringConstant=100
    alphaBoolBlend {blendOption=Linear}
    alphaScaleBiasClamp {scale=0.8}
    boundBone {boneName=dyn_pony_02}
    componentAppliedLinearAccClamp=<1100 1100 1100>
    componentLinearAccScale=<1 1 1>
    componentLinearVelScale=<1 1 1>
    gravityOverride {z=-900}
    physicsBodyDefinitions [{
      sphereCollisionRadius=0
      boundBone {boneName=dyn_pony_02}
      boxExtents=<8 20 15>
      constraintSetup {angularLimitsMax=<10 4 60>, angularLimitsMin=<-10 -90 -60>, angularTarget=<1 0 0>}
      localJointOffset=<7 0 0>
     }]
   }
   n3 = Anim Dynamics - Bone: dyn_pony_04(ComponentPose=@n2, Alpha=1, ExternalForce="0,0,0") {
    angularDampingOverride=1
    angularSpringConstant=1500
    bAngularSpring=true
    bGravityOverrideInSimSpace=true
    bOverrideAngularDamping=true
    bUseGravityOverride=true
    gravityScale=0.35
    linearSpringConstant=100
    alphaBoolBlend {blendOption=Linear}
    alphaScaleBiasClamp {scale=0.85}
    boundBone {boneName=dyn_pony_04}
    componentAppliedLinearAccClamp=<1100 1100 1100>
    componentLinearAccScale=<1 1 1>
    componentLinearVelScale=<1 1 1>
    gravityOverride {z=-900}
    physicsBodyDefinitions [{
      sphereCollisionRadius=0
      boundBone {boneName=dyn_pony_04}
      boxExtents=<8 20 15>
      constraintSetup {angularLimitsMax=<0 6 60>, angularLimitsMin=<0 -85 -60>, angularTarget=<1 0 0>}
      localJointOffset=<7 0 0>
     }]
   }
   Anim Dynamics - Bone: dyn_pony_05(ComponentPose=@n3, Alpha=1, ExternalForce="0,0,0") {
    angularDampingOverride=1
    angularSpringConstant=1500
    bAngularSpring=true
    bGravityOverrideInSimSpace=true
    bOverrideAngularDamping=true
    bUseGravityOverride=true
    gravityScale=0.35
    linearSpringConstant=100
    alphaBoolBlend {blendOption=Linear}
    alphaScaleBiasClamp {scale=0.95}
    boundBone {boneName=dyn_pony_05}
    componentAppliedLinearAccClamp=<1100 1100 1100>
    componentLinearAccScale=<1 1 1>
    componentLinearVelScale=<1 1 1>
    gravityOverride {z=-900}
    physicsBodyDefinitions [{
      sphereCollisionRadius=0
      boundBone {boneName=dyn_pony_05}
      boxExtents=<5 15 10>
      constraintSetup {angularLimitsMax=<0 20 60>, angularLimitsMin=<0 -85 -60>, angularTarget=<1 0 0>}
      localJointOffset=<4.5 0 0>
     }]
   }
   Control Rig() {
    alphaBoolBlend {blendOption=Linear}
    controlRigAssetReference {blueprintRigClass=/Game/Characters/Player/Female/Medium/Bodies/F_MED_Desert_Ops_Camo/Meshes/Parts/Hair/F_MED_Desert_Ops_Camo_Hair_CtrlRig.F_MED_Desert_Ops_Camo_Hair_CtrlRig_C}
   }
   Copy Pose From Mesh() {bUseAttachedParent=true}
   Copy Pose From Mesh() {bUseAttachedParent=true}
   "Transform (Modify) Bone - Bone: dyn_bang_c_02"(Alpha=0, Translation="0,0,0", Rotation="0.000,10,0.000", Scale="1,1,1") {
    rotationMode=BMM_Additive
    rotationSpace=BCS_BoneSpace
    translationMode=BMM_Additive
    translationSpace=BCS_BoneSpace
    alphaBoolBlend {blendOption=Linear}
    boneToModify {boneName=dyn_bang_c_02}
    rotation {yaw=10}
   }
   n4 = Get HeadRotBackAlpha() -> (HeadRotBackAlpha:float)
   "Transform (Modify) Bone - Bone: dyn_hair_bk_r_02"(Alpha=@n4, Translation="0,0,0", Rotation="0,0,0", Scale="0.5,0.5,0.5") {
    rotationMode=BMM_Additive
    rotationSpace=BCS_BoneSpace
    scaleMode=BMM_Additive
    scaleSpace=BCS_BoneSpace
    alphaBoolBlend {blendOption=Linear}
    boneToModify {boneName=dyn_hair_bk_r_02}
    scale=<0.5 0.5 0.5>
   }
   "Transform (Modify) Bone - Bone: dyn_parent_main_hair"(Alpha=1, Translation="0.000,-3.5,-0.000", Rotation="0.000,0.000,-4", Scale="1,1,1") {
    rotationMode=BMM_Additive
    rotationSpace=BCS_BoneSpace
    translationMode=BMM_Additive
    translationSpace=BCS_BoneSpace
    alphaBoolBlend {blendOption=Linear}
    boneToModify {boneName=dyn_parent_main_hair}
    rotation {roll=-4}
    translation {y=-3.5}
   }
   n5 = RigidBody(ComponentPose=Local To Component(LocalPose=Copy Pose From Mesh() {bUseAttachedParent=true}), Alpha=0.8) {
    bClampLinearTranslationLimitToRefPose=true
    bOverrideWorldGravity=true
    overridePhysicsAsset=&/F_MED_Desert_Ops_Camo_Ponytail_FaceAcc_Physics
    simulationSpace=BaseBoneSpace
    alphaBoolBlend {blendOption=Linear}
    baseBoneRef {boneName=root}
    componentAppliedLinearAccClamp=<600 600 1000>
    componentLinearAccScale=<0.5 0.5 0.35>
    componentLinearVelScale=<0.5 0.5 0.35>
    overrideWorldGravity {z=-980}
   }
   n6 = Get HeadRotBackAlpha()
   n7 = Look At - Bone: dyn_hair_bk_l_base(ComponentPose=Local To Component(LocalPose=Copy Pose From Mesh() {bUseAttachedParent=true}), Alpha=@n6) {
    lookAtClamp=100
    alphaBoolBlend {blendOption=Linear}
    boneToModify {boneName=dyn_hair_bk_l_base}
    lookAtLocation=<1 1 5>
    lookAtTarget {boneReference {boneName=spine_04}}
    lookAt_Axis {axis=<1 0>}
    lookUp_Axis {axis {x=0, z=1}}
   }
   n8 = Look At - Bone: dyn_hair_bk_l_01(ComponentPose=@n7, Alpha=@n6) {
    lookAtClamp=100
    alphaBoolBlend {blendOption=Linear}
    boneToModify {boneName=dyn_hair_bk_l_01}
    lookAtLocation=<1 1 5>
    lookAtTarget {boneReference {boneName=spine_04}}
    lookAt_Axis {axis=<1 0>}
    lookUp_Axis {axis {x=0, z=1}}
   }
   n9 = Look At - Bone: dyn_hair_bk_l_02(ComponentPose=@n8, Alpha=@n6) {
    lookAtClamp=100
    alphaBoolBlend {blendOption=Linear}
    boneToModify {boneName=dyn_hair_bk_l_02}
    lookAtLocation=<1 1 5>
    lookAtTarget {boneReference {boneName=spine_03}}
    lookAt_Axis {axis=<1 0>}
    lookUp_Axis {axis {x=0, z=1}}
   }
   n10 = Look At - Bone: dyn_hair_bk_mid_base(ComponentPose=@n9, Alpha=@n6) {
    lookAtClamp=100
    alphaBoolBlend {blendOption=Linear}
    boneToModify {boneName=dyn_hair_bk_mid_base}
    lookAtLocation=<1 1 5>
    lookAtTarget {boneReference {boneName=spine_04}}
    lookAt_Axis {axis=<1 0>}
    lookUp_Axis {axis {x=0, z=1}}
   }
   n11 = Look At - Bone: dyn_hair_bk_mid_01(ComponentPose=@n10, Alpha=@n6) {
    lookAtClamp=100
    alphaBoolBlend {blendOption=Linear}
    boneToModify {boneName=dyn_hair_bk_mid_01}
    lookAtLocation=<1 1 5>
    lookAtTarget {boneReference {boneName=spine_04}}
    lookAt_Axis {axis=<1 0>}
    lookUp_Axis {axis {x=0, z=1}}
   }
   n12 = Look At - Bone: dyn_hair_bk_02(ComponentPose=@n11, Alpha=@n6) {
    lookAtClamp=100
    alphaBoolBlend {blendOption=Linear}
    boneToModify {boneName=dyn_hair_bk_02}
    lookAtLocation=<1 1 5>
    lookAtTarget {boneReference {boneName=spine_04}}
    lookAt_Axis {axis=<1 0>}
    lookUp_Axis {axis {x=0, z=1}}
   }
   n13 = Look At - Bone: dyn_hair_base(ComponentPose=@n12, Alpha=Get HeadSwingAlpha() -> (HeadSwingAlpha:float)) {
    lookAtClamp=100
    alphaBoolBlend {blendOption=Linear}
    boneToModify {boneName=dyn_hair_base}
    lookAtLocation=<1 1 5>
    lookAtTarget {boneReference {boneName=lowerarm_l}}
    lookAt_Axis {axis=<1 0>}
    lookUp_Axis {axis {z=1}}
   }
   n14 = RigidBody(ComponentPose=@n13, Alpha=Get SkydivingBackAlpha() -> (SkydivingBackAlpha:float)) {
    bClampLinearTranslationLimitToRefPose=true
    bOverrideWorldGravity=true
    overridePhysicsAsset=/BRCosmetics/Characters/Player/Female/Medium/Bodies/F_MED_Rebirth_Medic/Meshes/Parts/F_MED_Rebirth_Medic_FaceAcc_Back_Physics
    simulationSpace=BaseBoneSpace
    alphaBoolBlend {blendOption=Linear}
    baseBoneRef {boneName=spine_05}
    componentAppliedLinearAccClamp=<1000 1000 1000>
    componentLinearAccScale=<0.25 0.25 0.25>
    componentLinearVelScale=<0.25 0.25 0.25>
    overrideWorldGravity {z=-980}
   }
   n15 = RigidBody(ComponentPose=@n14, Alpha=Get SkydivingFrontAlpha() -> (SkydivingFrontAlpha:float)) {
    bClampLinearTranslationLimitToRefPose=true
    bOverrideWorldGravity=true
    overridePhysicsAsset=/BRCosmetics/Characters/Player/Female/Medium/Bodies/F_MED_Rebirth_Medic/Meshes/Parts/F_MED_Rebirth_Medic_Frnt_FaceAcc_Physics_02
    simulationSpace=BaseBoneSpace
    alphaBoolBlend {blendOption=Linear}
    baseBoneRef {boneName=neck_01}
    componentAppliedLinearAccClamp=<1000 1000 1000>
    componentLinearAccScale=<0.5 0.5 0.25>
    componentLinearVelScale=<0.5 0.5 0.25>
    overrideWorldGravity {z=-980}
   }
   n16 = RigidBody(ComponentPose=@n15, Alpha=0.7) {
    bClampLinearTranslationLimitToRefPose=true
    bOverrideWorldGravity=true
    overridePhysicsAsset=&/F_MED_DesOpsCamo_Sides_FaceAcc_Physics
    simulationSpace=BaseBoneSpace
    alphaBoolBlend {blendOption=Linear}
    baseBoneRef {boneName=root}
    componentAppliedLinearAccClamp=<300 300 1050>
    componentLinearAccScale=<0.2 0.2 0.7>
    componentLinearVelScale=<0.2 0.2 0.7>
    overrideWorldGravity {z=-980}
   }
   n17 = RigidBody(ComponentPose=@n16, Alpha=0.7) {
    bClampLinearTranslationLimitToRefPose=true
    bOverrideWorldGravity=true
    overridePhysicsAsset=&/F_MED_Desert_Ops_Camo_Hair_Long_Physics
    simulationSpace=BaseBoneSpace
    alphaBoolBlend {blendOption=Linear}
    baseBoneRef {boneName=head}
    componentAppliedLinearAccClamp=<300 300 1050>
    componentLinearAccScale=<0.2 0.2 0.7>
    componentLinearVelScale=<0.2 0.2 0.7>
    overrideWorldGravity {z=-980}
   }
   n18 = "Constraint - dyn_parent_main_hair (T)"(ComponentPose=@n17, Alpha=1, ConstraintWeights_0=1) {
    alphaBoolBlend {blendOption=Linear}
    boneToModify {boneName=dyn_parent_main_hair}
    constraintSetup [{offsetOption=None, transformType=Translation, targetBone {boneName=dyn_hair_02}}]
    constraintWeights [1]
   }
   n19 = RigidBody(ComponentPose=@n18) {
    bClampLinearTranslationLimitToRefPose=true
    bOverrideWorldGravity=true
    overridePhysicsAsset=&/F_MED_DesOpsCamo_Tail_FaceAcc_Physics
    simulationSpace=BaseBoneSpace
    alphaBoolBlend {blendOption=Linear}
    baseBoneRef {boneName=root}
    componentAppliedLinearAccClamp=<600 600 600>
    componentLinearAccScale=<0.5 0.5 0.35>
    componentLinearVelScale=<0.5 0.5 0.35>
    overrideWorldGravity {z=-980}
   }
   n20 = Get HeadRotBackAlpha()
   n21 = "Transform (Modify) Bone - Bone: dyn_hair_bk_l_01"(ComponentPose=@n19, Alpha=@n20, Translation="0,0,0", Rotation="10,20,0", Scale="0.8,0.8,0.8") {
    rotationMode=BMM_Additive
    rotationSpace=BCS_BoneSpace
    scaleMode=BMM_Additive
    scaleSpace=BCS_BoneSpace
    alphaBoolBlend {blendOption=Linear}
    boneToModify {boneName=dyn_hair_bk_l_01}
    rotation {pitch=10, yaw=20}
    scale=<0.8 0.8 0.8>
   }
   n22 = "Transform (Modify) Bone - Bone: dyn_hair_bk_mid_01"(ComponentPose=@n21, Alpha=@n20, Translation="0,0,0", Rotation="30,20,0", Scale="0.8,0.8,0.8") {
    rotationMode=BMM_Additive
    rotationSpace=BCS_BoneSpace
    scaleMode=BMM_Additive
    scaleSpace=BCS_BoneSpace
    alphaBoolBlend {blendOption=Linear}
    boneToModify {boneName=dyn_hair_bk_mid_01}
    rotation {pitch=30, yaw=20}
    scale=<0.8 0.8 0.8>
   }
   n23 = "Transform (Modify) Bone - Bone: dyn_hair_bk_r_01"(ComponentPose=@n22, Alpha=@n4, Translation="0,0,0", Rotation="0,0,0", Scale="0.5,0.5,0.5") {
    rotationMode=BMM_Additive
    rotationSpace=BCS_BoneSpace
    scaleMode=BMM_Additive
    scaleSpace=BCS_BoneSpace
    alphaBoolBlend {blendOption=Linear}
    boneToModify {boneName=dyn_hair_bk_r_01}
    scale=<0.5 0.5 0.5>
   }
   n24 = "Transform (Modify) Bone - Bone: dyn_hair_bk_l_01"(ComponentPose=@n23, Alpha=1, Translation="2.239,-1.363,1.7", Rotation="0,0,0", Scale="1,1,1") {
    rotationMode=BMM_Additive
    rotationSpace=BCS_BoneSpace
    translationMode=BMM_Additive
    translationSpace=BCS_BoneSpace
    alphaBoolBlend {blendOption=Linear}
    boneToModify {boneName=dyn_hair_bk_l_01}
    translation=<2.239 -1.363 1.7>
   }
   n25 = "Transform (Modify) Bone - Bone: dyn_hair_bk_mid_01"(ComponentPose=@n24, Alpha=1, Translation="1.5,-1,-0.5", Rotation="0,0,0", Scale="1,1,1") {
    rotationMode=BMM_Additive
    rotationSpace=BCS_BoneSpace
    translationMode=BMM_Additive
    translationSpace=BCS_BoneSpace
    alphaBoolBlend {blendOption=Linear}
    boneToModify {boneName=dyn_hair_bk_mid_01}
    translation=<1.5 -1 -0.5>
   }
   n26 = Get no_vest_alpha() -> (no_vest_alpha:double)
   n27 = "Transform (Modify) Bone - Bone: dyn_hair_base"(ComponentPose=@n25, Alpha=@n26, Translation="-1.85,0.115,-0.936", Rotation="7.653,0.052,-0.412", Scale="1,1,1") {
    rotationMode=BMM_Additive
    rotationSpace=BCS_BoneSpace
    translationMode=BMM_Additive
    translationSpace=BCS_BoneSpace
    alphaBoolBlend {blendOption=Linear}
    boneToModify {boneName=dyn_hair_base}
    rotation=<rot 7.653 0.052 -0.412>
    translation=<-1.85 0.115 -0.936>
   }
   n28 = "Transform (Modify) Bone - Bone: dyn_parent_main_hair"(ComponentPose=@n27, Alpha=@n26, Translation="-0.000,-2.423,-2.466", Rotation="0.000,0.000,-1.725", Scale="1,1,1") {
    rotationMode=BMM_Additive
    rotationSpace=BCS_BoneSpace
    translationMode=BMM_Additive
    translationSpace=BCS_BoneSpace
    alphaBoolBlend {blendOption=Linear}
    boneToModify {boneName=dyn_parent_main_hair}
    rotation {roll=-1.725}
    translation {y=-2.423, z=-2.466}
   }
   n29 = Get no_hat_short_alpha() -> (no_hat_short_alpha:double)
   n30 = Anim Dynamics - Bone: dyn_bang_r_01(ComponentPose=Local To Component(LocalPose=Copy Pose From Mesh() {bUseAttachedParent=true}), Alpha=@n29, ExternalForce="0,0,0") {
    angularDampingOverride=1
    angularSpringConstant=100
    bGravityOverrideInSimSpace=true
    bOverrideAngularDamping=true
    bUseGravityOverride=true
    gravityScale=0.35
    alphaBoolBlend {blendOption=Linear}
    boundBone {boneName=dyn_bang_r_01}
    componentAppliedLinearAccClamp=<1100 1100 1100>
    componentLinearAccScale=<1 1 1>
    componentLinearVelScale=<1 1 1>
    gravityOverride {z=-980}
    physicsBodyDefinitions [{
      sphereCollisionRadius=0
      boundBone {boneName=dyn_bang_r_01}
      boxExtents=<20 20 20>
      constraintSetup {angularLimitsMax=<0 3 10>, angularLimitsMin=<0 -35 -35>, angularTarget=<0 0 0>}
      localJointOffset=<13 0 0>
     }]
   }
   n31 = Anim Dynamics - Bone: dyn_bang_r_02(ComponentPose=@n30, Alpha=@n29, ExternalForce="0,0,0") {
    angularDampingOverride=1
    angularSpringConstant=100
    bGravityOverrideInSimSpace=true
    bOverrideAngularDamping=true
    bUseGravityOverride=true
    gravityScale=0.35
    alphaBoolBlend {blendOption=Linear}
    boundBone {boneName=dyn_bang_r_02}
    componentAppliedLinearAccClamp=<1100 1100 1100>
    componentLinearAccScale=<1 1 1>
    componentLinearVelScale=<1 1 1>
    gravityOverride {z=-980}
    physicsBodyDefinitions [{
      sphereCollisionRadius=0
      boundBone {boneName=dyn_bang_r_02}
      boxExtents=<12 19 19>
      constraintSetup {angularLimitsMax=<0 5 25>, angularLimitsMin=<0 -35 -35>, angularTarget=<0 0 0>}
      localJointOffset=<10 0 0>
     }]
   }
   n32 = Anim Dynamics - Bone: dyn_bang_r_03(ComponentPose=@n31, Alpha=@n29, ExternalForce="0,0,0") {
    angularDampingOverride=1
    angularSpringConstant=100
    bGravityOverrideInSimSpace=true
    bOverrideAngularDamping=true
    bUseGravityOverride=true
    gravityScale=0.35
    alphaBoolBlend {blendOption=Linear}
    boundBone {boneName=dyn_bang_r_03}
    componentAppliedLinearAccClamp=<1100 1100 1100>
    componentLinearAccScale=<1 1 1>
    componentLinearVelScale=<1 1 1>
    gravityOverride {z=-980}
    physicsBodyDefinitions [{
      sphereCollisionRadius=0
      boundBone {boneName=dyn_bang_r_03}
      boxExtents=<9 19 19>
      constraintSetup {angularLimitsMax=<0 10 40>, angularLimitsMin=<0 -35 -45>, angularTarget=<0 0 0>}
      localJointOffset=<8 0 0>
     }]
   }
   n33 = Get no_hat_short_alpha()
   n34 = Anim Dynamics - Bone: dyn_bang_l_01(ComponentPose=@n32, Alpha=@n33, ExternalForce="0,0,0") {
    angularDampingOverride=1
    angularSpringConstant=100
    bGravityOverrideInSimSpace=true
    bOverrideAngularDamping=true
    bUseGravityOverride=true
    gravityScale=0.35
    alphaBoolBlend {blendOption=Linear}
    boundBone {boneName=dyn_bang_l_01}
    componentAppliedLinearAccClamp=<1100 1100 1100>
    componentLinearAccScale=<1 1 1>
    componentLinearVelScale=<1 1 1>
    gravityOverride {z=-980}
    physicsBodyDefinitions [{
      sphereCollisionRadius=0
      boundBone {boneName=dyn_bang_l_01}
      boxExtents=<20 20 20>
      constraintSetup {angularLimitsMax=<0 3 30>, angularLimitsMin=<0 -35 -5>, angularTarget=<0 0 0>}
      localJointOffset=<13 0 0>
     }]
   }
   n35 = Anim Dynamics - Bone: dyn_bang_l_02(ComponentPose=@n34, Alpha=@n33, ExternalForce="0,0,0") {
    angularDampingOverride=1
    angularSpringConstant=100
    bGravityOverrideInSimSpace=true
    bOverrideAngularDamping=true
    bUseGravityOverride=true
    gravityScale=0.35
    alphaBoolBlend {blendOption=Linear}
    boundBone {boneName=dyn_bang_l_02}
    componentAppliedLinearAccClamp=<1100 1100 1100>
    componentLinearAccScale=<1 1 1>
    componentLinearVelScale=<1 1 1>
    gravityOverride {z=-980}
    physicsBodyDefinitions [{
      sphereCollisionRadius=0
      boundBone {boneName=dyn_bang_l_02}
      boxExtents=<12 19 19>
      constraintSetup {angularLimitsMax=<0 5 35>, angularLimitsMin=<0 -35 -25>, angularTarget=<0 0 0>}
      localJointOffset=<10 0 0>
     }]
   }
   n36 = Anim Dynamics - Bone: dyn_bang_l_03(ComponentPose=@n35, Alpha=@n33, ExternalForce="0,0,0") {
    angularDampingOverride=1
    angularSpringConstant=100
    bGravityOverrideInSimSpace=true
    bOverrideAngularDamping=true
    bUseGravityOverride=true
    gravityScale=0.35
    alphaBoolBlend {blendOption=Linear}
    boundBone {boneName=dyn_bang_l_03}
    componentAppliedLinearAccClamp=<1100 1100 1100>
    componentLinearAccScale=<1 1 1>
    componentLinearVelScale=<1 1 1>
    gravityOverride {z=-980}
    physicsBodyDefinitions [{
      sphereCollisionRadius=0
      boundBone {boneName=dyn_bang_l_03}
      boxExtents=<9 19 19>
      constraintSetup {angularLimitsMax=<0 10 40>, angularLimitsMin=<0 -35 -45>, angularTarget=<0 0 0>}
      localJointOffset=<8 0 0>
     }]
   }
   n37 = Get helmet_hairc_alpha() -> (helmet_hairc_alpha:double)
   n38 = Anim Dynamics - Bone: dyn_bang_c_01(ComponentPose=@n36, Alpha=@n37, ExternalForce="0,0,0") {
    angularDampingOverride=1
    angularSpringConstant=5000
    bGravityOverrideInSimSpace=true
    bOverrideAngularDamping=true
    bUseGravityOverride=true
    gravityScale=0.35
    linearSpringConstant=100
    alphaBoolBlend {blendOption=Linear}
    alphaScaleBiasClamp {scale=0.5}
    boundBone {boneName=dyn_bang_c_01}
    componentAppliedLinearAccClamp=<1100 1100 1100>
    componentLinearAccScale=<1 1 1>
    componentLinearVelScale=<1 1 1>
    gravityOverride {z=-980}
    physicsBodyDefinitions [{
      sphereCollisionRadius=0
      boundBone {boneName=dyn_bang_c_01}
      boxExtents=<10 20 5>
      constraintSetup {angularLimitsMax=<0 6 5>, angularLimitsMin=<0 -35 -2>, angularTarget=<1 0 0>}
      localJointOffset=<8 0 0>
     }]
   }
   n39 = Anim Dynamics - Bone: dyn_bang_c_02(ComponentPose=@n38, Alpha=@n37, ExternalForce="0,0,0") {
    angularDampingOverride=1
    angularSpringConstant=100
    bGravityOverrideInSimSpace=true
    bOverrideAngularDamping=true
    bUseGravityOverride=true
    gravityScale=0.35
    alphaBoolBlend {blendOption=Linear}
    alphaScaleBiasClamp {scale=0.6}
    boundBone {boneName=dyn_bang_c_02}
    componentAppliedLinearAccClamp=<1100 1100 1100>
    componentLinearAccScale=<1 1 1>
    componentLinearVelScale=<1 1 1>
    gravityOverride {z=-980}
    physicsBodyDefinitions [{
      sphereCollisionRadius=0
      boundBone {boneName=dyn_bang_c_02}
      boxExtents=<15 20 5>
      constraintSetup {angularLimitsMax=<0 9 15>, angularLimitsMin=<0 -35 -2>, angularTarget=<0 0 0>}
      localJointOffset=<8 0 0>
     }]
   }
   n40 = Anim Dynamics - Bone: dyn_bang_c_03(ComponentPose=@n39, Alpha=@n37, ExternalForce="0,0,0") {
    angularDampingOverride=1
    angularSpringConstant=2500
    bAngularSpring=true
    bGravityOverrideInSimSpace=true
    bOverrideAngularDamping=true
    bUseGravityOverride=true
    gravityScale=0.35
    alphaBoolBlend {blendOption=Linear}
    alphaScaleBiasClamp {scale=0.6}
    boundBone {boneName=dyn_bang_c_03}
    componentAppliedLinearAccClamp=<1100 1100 1100>
    componentLinearAccScale=<1 1 1>
    componentLinearVelScale=<1 1 1>
    gravityOverride {z=-980}
    physicsBodyDefinitions [{
      sphereCollisionRadius=0
      boundBone {boneName=dyn_bang_c_03}
      boxExtents=<15 20 5>
      constraintSetup {angularLimitsMax=<0 5 60>, angularLimitsMin=<0 -35 -5>, angularTarget=<0 0 0>}
      localJointOffset=<4.5 0 0>
     }]
   }
   n41 = Blend Poses by int!(
    BlendPose_0:Pose Link Structure=Copy Pose From Mesh() {bUseAttachedParent=true}
    BlendPose_1:Pose Link Structure=Component To Local(ComponentPose=@n28)
    BlendPose_2:Pose Link Structure=Component To Local(ComponentPose=@n40)
    BlendPose_3:Pose Link Structure=Blend Poses by bool(
     BlendPose_0=Copy Pose From Mesh() {bUseAttachedParent=true}
     BlendPose_1=Component To Local(ComponentPose=@n5)
     bActiveValue=Get hat_active() -> (hat_active:bool)
    ) {blendType=Linear}
    BlendTime_0:float=0.1
    BlendTime_1:float=0.1
    BlendTime_2:float=0.1
    BlendTime_3:float=0.1
    ActiveChildIndex:int=Get active_hair_index() -> (active_hair_index:int=0)
   ) -> (Pose:Pose Link Structure) {
    is=title AnimGraphNode_BlendListByInt
    blendType=Linear
    blendPose [{linkId=-1, sourceLinkId=-1}, {linkId=-1, sourceLinkId=-1}, {linkId=-1, sourceLinkId=-1}, {linkId=-1, sourceLinkId=-1}]
    blendTime [0.1,0.1,0.1,0.1]
   }
   Output Pose(Result=@n41) {layerGroup=None}
  ]}
 EventGraph {nodes [
   Event OnCharacterCustomizationLoaded()
   > n1 = Sequence!(execute) -> (then_0, then_1, then_2:Exec, then_3:Exec, then_4:Exec, then_5:Exec)
   Branch(execute=@n1.then_2, Condition=CharacterItemDefinitionHasAnyMetaTag(Tags=Make Array("[0]"=(TagName="Cosmetics.MetaTags.Hair.00"))))
   > Set active_hair_index(active_hair_index:int=1) -> (Output_Get:int=0)
   Branch(execute=@n1.then_5, Condition=CharacterItemDefinitionHasAnyMetaTag(Tags=Make Array("[0]"=(TagName="Cosmetics.MetaTags.Hair.03"))))
   > Set active_hair_index(active_hair_index=0) -> (Output_Get=0)
   Branch(execute=@n1.then_3, Condition=CharacterItemDefinitionHasAnyMetaTag(Tags=Make Array("[0]"=(TagName="Cosmetics.MetaTags.Hair.01"))))
   > Set active_hair_index(active_hair_index=2) -> (Output_Get=0)
   Branch(execute=@n1.then_4, Condition=CharacterItemDefinitionHasAnyMetaTag(Tags=Make Array("[0]"=(TagName="Cosmetics.MetaTags.Hair.02"))))
   > Set active_hair_index(active_hair_index=3) -> (Output_Get=0)
   n2 = Branch(
    execute=@n1.then_0
    Condition=CharacterItemDefinitionHasAnyMetaTag(Tags=Make Array!(
      "[0]"=(TagName="Cosmetics.MetaTags.Hat.01")
      "[1]":Gameplay Tag Structure=(TagName="Cosmetics.MetaTags.Hat.02")
      "[2]":Gameplay Tag Structure=(TagName="Cosmetics.MetaTags.Hat.03")
      "[3]":Gameplay Tag Structure=(TagName="Cosmetics.MetaTags.Hat.04")
     ) -> (Array))
   )
   > Set hat_active(hat_active:bool=true) -> (Output_Get:bool)
   > Set no_hat_alpha(no_hat_alpha:double=0) -> (Output_Get:double)
   > Set no_hat_short_alpha(no_hat_short_alpha:double=0.25) -> (Output_Get:double)
   > n3 = Branch(Condition=CharacterItemDefinitionHasAnyMetaTag(Tags=Make Array!("[0]"=(TagName="Cosmetics.MetaTags.Hat.01"), "[1]"=(TagName="Cosmetics.MetaTags.Hat.02")) -> (Array)))
   > Set helmet_hairc_alpha(helmet_hairc_alpha:double=0) -> (Output_Get:double)
   Set helmet_hairc_alpha(execute=@n3.else, helmet_hairc_alpha=0.1)
   Set hat_active(execute=@n2.else, hat_active=false)
   > Set no_hat_alpha(no_hat_alpha=1)
   > Set no_hat_short_alpha(no_hat_short_alpha=0.55)
   n4 = Branch(
    execute=@n1.then_1
    Condition=CharacterItemDefinitionHasAnyMetaTag(Tags=Make Array!(
      "[0]"=(TagName="Cosmetics.MetaTags.Vest.01")
      "[1]"=(TagName="Cosmetics.MetaTags.Vest.02")
      "[2]":Gameplay Tag Structure=(TagName="Cosmetics.MetaTags.Vest.03")
      "[3]":Gameplay Tag Structure=(TagName="Cosmetics.MetaTags.Vest.04")
     ) -> (Array))
   )
   > Set vest_active(vest_active:bool=true) -> (Output_Get:bool)
   > Set no_vest_alpha(no_vest_alpha:double=0) -> (Output_Get:double)
   Set vest_active(execute=@n4.else, vest_active=false)
   > Set no_vest_alpha(no_vest_alpha=1)
  ]}
]
variables [
 active_hair_index {type=integer, default=1}
 hat_active {type=boolean}
 helmet_hairc_alpha {type=number}
 no_hat_alpha {type=number}
 no_hat_short_alpha {type=number}
 no_vest_alpha {type=number}
 vest_active {type=boolean}
]
